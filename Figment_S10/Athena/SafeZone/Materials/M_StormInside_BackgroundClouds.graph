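# Translucent unlit material for the storm's background cloud layer in Fortnite's Athena safe zone.
# It projects a cloud mask triplanar-style along view direction at a set distance, panning two planes and crossfading scale-animated samples on the third.
# The mask drives color and opacity, with horizon tint, sun-direction highlight and cutout, and depth fade.
# Further logic provides camera-attached vertex offsets and fades from camera position against safe-zone collection values or a cone shape.
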Material /Figment_S10/Athena/SafeZone/Materials/M_StormInside_BackgroundClouds
graph {
 properties {
  bUseLightmapDirectionality=false
  bUseTranslucencyVertexFog=false
  blendMode=BLEND_Translucent
  pixelDepthOffsetMode=PDOM_Legacy
  shadingModel=MSM_Unlit
  usedShadingModels=Unlit
 }
 nodes [
  n1 = Constant3Vector() {constant=<rgba 0 0 1 0>}
  n2 = WorldPosition()
  n3 = CameraPositionWS()
  n4 = Normalize(VectorInput=Subtract(A=@n2.XYZ, B=@n3))
  n5 = OneMinus(None=Saturate(None=DotProduct(A=@n1, B=@n4)))
  n6 = ScalarParameter() {defaultValue=10, parameterName=Horizon FallOFF}
  n7 = VectorParameter() {parameterName=Color_4, defaultValue=<rgba 0.0631 0.0247 0.1823 1>}
  n8 = ComponentMask(None=@n4) {b=true}
  n9 = Add(A=@n8)
  n10 = LinearInterpolate(A=VectorParameter() {parameterName=Color_3, defaultValue=<rgba 0.005 0.0055 0.0208 1>}.RGB, B=@n7.RGB, Alpha=@n9)
  n11 = QualitySwitch(Default=@n10, Low=@n7.RGB, High=@n10, Medium=@n10)
  n12 = VectorParameter() {parameterName=Color_2, defaultValue=<rgba 0.2941 0.084 0.375 1>}
  n13 = CollectionParameter() {
   collection=&FortniteMaterialParameters
   parameterName=SunlightDirectionalVector
  }
  n14 = Clamp(None=DotProduct(A=@n4, B=ComponentMask(None=@n13) {b=true, g=true, r=true}))
  n15 = Subtract(A=ScalarParameter() {defaultValue=21379.3926, parameterName=ProjectionDistance}, B=Distance(A=@n2.XYZ, B=@n3))
  n16 = Subtract(A=Add(A=@n2.XYZ, B=Multiply(A=@n15, B=@n4)), B=@n3)
  n17 = ComponentMask(None=@n16) {r=true}
  n18 = ComponentMask(None=@n16) {b=true}
  n19 = AppendVector(A=@n17, B=@n18)
  n20 = ScalarParameter() {defaultValue=-20293.4453, parameterName=Scale}
  n21 = Multiply(A=Time() {period=5}, B=ScalarParameter() {defaultValue=0.5524, parameterName=Speed})
  n22 = Multiply(A=VectorParameter() {parameterName=PanDirection, defaultValue=<rgba -1 0 0.1143 1>}.RGB, B=@n21)
  n23 = TextureSampleParameter2D(UVs=Add(A=Divide(A=@n19, B=@n20), B=ComponentMask(None=@n22) {b=true, r=true})) {parameterName=CloudTexture, samplerType=SAMPLERTYPE_Masks, texture=/Game/Athena/SafeZone/Textures/StormCloudtubes_M}
  n24 = Multiply(A=Constant3Vector() {constant=<rgba 0 1 0 0>}, B=@n23.B)
  n25 = ComponentMask(None=@n16) {g=true}
  n26 = AppendVector(A=@n25, B=@n18)
  n27 = ScalarParameter() {defaultValue=-4000, parameterName=Scale_2}
  n28 = ScalarParameter() {defaultValue=-33036.5312, parameterName=Scale_1}
  n29 = Multiply(A=Time() {period=5}, B=ScalarParameter() {defaultValue=0.2286, parameterName=Speed_1})
  n30 = Frac(None=@n29)
  n31 = ComponentMask(None=@n4) {r=true}
  n32 = Clamp(None=Sign(None=@n31))
  n33 = LinearInterpolate(A=@n27, B=@n28, Alpha=LinearInterpolate(A=OneMinus(None=@n30), B=@n30, Alpha=@n32))
  n34 = Frac(None=Add(A=@n29) {constB=0.5})
  n35 = OneMinus(None=@n34)
  n36 = Divide(A=@n26, B=LinearInterpolate(A=@n27, B=@n28, Alpha=LinearInterpolate(A=@n35, B=@n34, Alpha=@n32)))
  n37 = Constant2Vector() {g=0.743, r=0.342}
  n38 = Add(A=Cosine(None=@n29))
  n39 = LinearInterpolate(
   A=TextureSampleParameter2D(UVs=Divide(A=@n26, B=@n33)) {parameterName=CloudTexture, samplerType=SAMPLERTYPE_Masks, texture=&StormCloudtubes_M}.B
   B=TextureSampleParameter2D(UVs=Add(A=@n36, B=@n37)) {parameterName=CloudTexture, samplerType=SAMPLERTYPE_Masks, texture=&StormCloudtubes_M}.B
   Alpha=Clamp(None=Divide(A=@n38))
  )
  n40 = Multiply(A=@n39, B=Constant3Vector() {constant=<rgba 0 0 1 0>})
  n41 = AppendVector(A=@n25, B=@n17)
  n42 = ScalarParameter() {defaultValue=-29100.7031, parameterName=Scale_5}
  n43 = Multiply(A=@n21, B=VectorParameter() {parameterName=PanDirection2_1, defaultValue=<rgba 0 -0.6762 0 1>}.RGB)
  n44 = TextureSampleParameter2D(UVs=Add(A=Divide(A=@n41, B=@n42), B=ComponentMask(None=@n43) {g=true, r=true})) {parameterName=CloudTexture, samplerType=SAMPLERTYPE_Masks, texture=&StormCloudtubes_M}
  n45 = Constant3Vector() {constant=<rgba 1 0 0 0>}
  n46 = LinearInterpolate(
   A=LinearInterpolate(
    A=StaticSwitchParameter(True=@n24, False=@n23.B) {parameterName=Deubg Directions_2}
    B=StaticSwitchParameter(True=@n40, False=@n39) {parameterName=Deubg Directions_3}
    Alpha=Abs(None=@n31)
   )
   B=StaticSwitchParameter(True=Multiply(A=@n44.B, B=@n45), False=@n44.B) {parameterName=Deubg Directions_1}
   Alpha=Abs(None=@n8)
  )
  Add(
   A=Multiply(A=VectorParameter() {parameterName=HorizonColor, defaultValue=<rgba 0.0667 0.2 0.1711 1>}.RGB, B=Power(Base=@n5, Exp=@n6))
   B=LinearInterpolate(
    A=VectorParameter() {parameterName=Color_1, defaultValue=<rgba 0.0298 0.0004 0.04 1>}.RGB
    B=LinearInterpolate(A=@n11, B=@n12.RGB, Alpha=@n14)
    Alpha=OneMinus(None=@n46)
   )
  )
  n47 = ObjectPositionWS()
  n48 = Distance(A=@n47, B=CameraPositionWS())
  n49 = ScalarParameter() {defaultValue=5000, parameterName=DrawRadius}
  n50 = Saturate(None=Divide(A=Subtract(A=@n48, B=@n49), B=ScalarParameter() {defaultValue=1, parameterName=DrawDistanceFade}))
  Ceil(None=@n50) {desc=floor}
  n51 = MaterialFunctionCall!(
   Object Basis 1
   Object Basis 2
   Object Basis 3
   Camera Space Position Offset=VectorParameter() {parameterName=ScreenPosition, defaultValue=<rgba 0 0 22 1>}.RGB
  ) {materialFunction=/Engine/Functions/Engine_MaterialFunctions02/WorldPositionOffset/AttachMeshToTheCamera}
  n52 = Subtract(A=@n47, B=WorldPosition().XYZ)
  LinearInterpolate(A=@n51.World Position Offset, B=@n52, Alpha=Floor(None=@n50))
  n53 = Power(Base=@n14, Exp=ScalarParameter() {defaultValue=42.9937, parameterName=Falloff})
  n54 = LinearInterpolate(
   A=ScalarParameter() {defaultValue=-0.1748, parameterName=Opacity_1}
   B=ScalarParameter() {defaultValue=1.5592, parameterName=Opacity_2}
   Alpha=@n46
  )
  n55 = DepthFade(Opacity=@n54, FadeDistance=ScalarParameter() {defaultValue=2740.7466, parameterName=DepthFade})
  Multiply(A=OneMinus(None=Saturate(None=@n53)), B=FeatureLevelSwitch(Default=@n55, ES2=@n54, ES3_1=@n54, SM4_REMOVED=@n55))
  ShadingPathSwitch()
  n56 = Ceil(None=@n50)
  n57 = Add(A=Multiply(A=OneMinus(None=@n56), B=Constant3Vector() {constant=<rgba 0 0 -54654566400 0>}), B=@n51.World Position Offset)
  n58 = Add(A=@n57, B=Multiply(A=VertexNormalWS(), B=ScalarParameter() {defaultValue=-3714.3762, parameterName=MobileOffsetPush}))
  StaticSwitchParameter(True=@n52, False=FeatureLevelSwitch(Default=@n57, ES2=@n58, ES3_1=@n58, SM4_REMOVED=@n57)) {parameterName=IsGroundLevel}
  n59 = ScalarParameter() {defaultValue=0.3619, parameterName=Opacity_3}
  n60 = ScalarParameter() {defaultValue=2.3278, parameterName=Opacity_4}
  StaticSwitchParameter(True=@n54, False=LinearInterpolate(A=@n54, B=LinearInterpolate(A=@n59, B=@n60), Alpha=Divide(A=@n9))) {parameterName=IsGroundLevel}
  StaticSwitchParameter(True=@n51.World Position Offset, False=Constant3Vector()) {defaultValue=true, parameterName=AttachToCamera}
  n61 = Normalize(VectorInput=Transform(None=Constant3Vector() {constant=<rgba 0 0 1 0>}) {transformSourceType=TRANSFORMSOURCE_Local})
  n62 = ObjectPositionWS()
  n63 = Subtract(A=DotProduct(A=CameraPositionWS(), B=@n61), B=DotProduct(A=@n62, B=@n61))
  n64 = ScalarParameter() {defaultValue=200, parameterName=F_Height}
  n65 = Divide(A=Subtract(A=@n63, B=@n64), B=ScalarParameter() {defaultValue=400, parameterName=F_BlendTop})
  n66 = Divide(A=Subtract(A=Multiply(A=@n63) {constB=-1}, B=@n64), B=ScalarParameter() {defaultValue=400, parameterName=F_BlendBottom})
  n67 = Multiply(A=OneMinus(None=Saturate(None=@n65)), B=OneMinus(None=Saturate(None=@n66)))
  n68 = Multiply(A=@n63, B=@n61)
  n69 = ScalarParameter() {defaultValue=200, parameterName=F_Height}
  n70 = Divide(A=Add(A=@n63, B=@n69), B=Multiply(A=@n69) {constB=2})
  n71 = Subtract(
   A=Distance(A=Add(A=@n68, B=@n62), B=CameraPositionWS())
   B=LinearInterpolate(
    A=ScalarParameter() {defaultValue=500, parameterName=F_RadiusBottom}
    B=ScalarParameter() {defaultValue=500, parameterName=F_RadiusTop}
    Alpha=Saturate(None=@n70)
   ) {constAlpha=0}
  )
  n72 = ScalarParameter() {defaultValue=500, parameterName=F_BlendRadius}
  n73 = OneMinus(None=Saturate(None=Divide(A=@n71, B=@n72)))
  n74 = ComponentMask(None=CameraPositionWS()) {b=true}
  n75 = StaticSwitchParameter(
   True=CollectionParameter() {collection=/Game/Athena/SafeZone/MPC_SafeZoneClouds, parameterName=SafeZoneCapHeight}
   False=ScalarParameter() {defaultValue=35142.25, parameterName=Heightoffset}
  ) {parameterName=IsSkyCap}
  n76 = Clamp(None=Divide(A=Subtract(A=@n74, B=@n75), B=ScalarParameter() {defaultValue=5000, parameterName=Heightoffset_1}))
  n77 = Distance(
   A=ComponentMask(None=CollectionParameter() {collection=&MPC_SafeZoneClouds, parameterName=SafeZoneLoc}) {g=true, r=true}
   B=ComponentMask(None=CameraPositionWS()) {g=true, r=true}
  )
  n78 = LinearInterpolate(
   A=CollectionParameter() {collection=&MPC_SafeZoneClouds, parameterName=SafeZoneScale}
   Alpha=CollectionParameter() {collection=&MPC_SafeZoneClouds, parameterName=PreventStormShrinking}
  )
  n79 = Multiply(A=CollectionParameter() {collection=&MPC_SafeZoneClouds, parameterName=CloudStartingBias}) {constB=2}
  n80 = Add(
   A=LinearInterpolate(
    A=ScalarParameter() {parameterName=Ring_Radius_Min}
    B=ScalarParameter() {defaultValue=212768.9688, parameterName=Ring_Radius_Max}
    Alpha=Add(A=@n78, B=@n79)
   )
   B=ScalarParameter() {defaultValue=-36555.5625, parameterName=Fade_Bias}
  )
  n81 = Clamp(None=Divide(A=Subtract(A=@n77, B=@n80), B=ScalarParameter() {defaultValue=20837.6094, parameterName=Fade_Distance}))
  n82 = StaticSwitchParameter(True=OneMinus(None=@n81), False=@n81) {parameterName=IsSkyCap}
  VertexInterpolator(VS=StaticSwitchParameter(True=Multiply(A=@n67, B=@n73), False=Multiply(A=@n76, B=@n82)) {parameterName=HasConeShapeFalloff})
  VertexNormalWS()
 ]
}
